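# Sound cue for the Knight Remix pickaxe swing, a cosmetic melee weapon sound.
# A mixer layers two branches: Knight Remix swings under close attenuation and shared fire-axe distant swings under medium attenuation.
# Each branch runs through a modulator and a quality-level switch that shrinks the random pool to fewer variants, then a single wave, at different quality levels.
# Per-owner concurrency override stops the oldest swing, and the cue primes on load.
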SoundCue /BRCosmetics/Sounds/PickAxes/Knight_Remix/PickAxe_Knight_Remix_Swing_Cue
root=Mixer_0
nodes [
 {id=Attenuation_0, children [Modulator_1], properties {attenuationSettings=/Game/Sounds/Attenuation/Fort_Weapon_Pickaxe_Close_Att}}
 {id=Attenuation_1, children [Modulator_0], properties {attenuationSettings=/Game/Sounds/Attenuation/Fort_Weapon_Pickaxe_Medium_Att}}
 {id=Mixer_0, children [Attenuation_0, Attenuation_1], properties {inputVolume [1,1]}}
 {id=Modulator_0, children [QualityLevel_1]}
 {id=Modulator_1, children [QualityLevel_0]}
 {id=QualityLevel_0, children [Random_2, Random_8, WavePlayer_28, WavePlayer_28]}
 {id=QualityLevel_1, children [Random_0, Random_1, WavePlayer_12, WavePlayer_12]}
 {id=Random_0, children [WavePlayer_1,WavePlayer_2,WavePlayer_4,WavePlayer_6,WavePlayer_7,WavePlayer_8], properties {weights [1,1,1,1,1,1]}}
 {id=Random_1, children [WavePlayer_10, WavePlayer_9, WavePlayer_11], properties {weights [1,1,1]}}
 {
  id=Random_2
  children [WavePlayer_0, WavePlayer_3, WavePlayer_5, WavePlayer_14, WavePlayer_15, WavePlayer_17]
  properties {weights [1,1,1,1,1,1]}
 }
 {id=Random_8, children [WavePlayer_22, WavePlayer_23, WavePlayer_24], properties {weights [1,1,1]}}
 {id=WavePlayer_0, properties {soundWaveAssetPtr=&/PA_Knight_Remix_Swing_01}}
 {id=WavePlayer_1, properties {soundWaveAssetPtr=/Game/Athena/Sounds/Weapons/Fire_Axe_Swing_Distant_04}}
 {id=WavePlayer_10, properties {soundWaveAssetPtr=&Fire_Axe_Swing_Distant_04}}
 {id=WavePlayer_11, properties {soundWaveAssetPtr=/Game/Athena/Sounds/Weapons/Fire_Axe_Swing_Distant_06}}
 {id=WavePlayer_12, properties {soundWaveAssetPtr=&Fire_Axe_Swing_Distant_06}}
 {id=WavePlayer_14, properties {soundWaveAssetPtr=&/PA_Knight_Remix_Swing_04}}
 {id=WavePlayer_15, properties {soundWaveAssetPtr=&/PA_Knight_Remix_Swing_05}}
 {id=WavePlayer_17, properties {soundWaveAssetPtr=&/PA_Knight_Remix_Swing_06}}
 {id=WavePlayer_2, properties {soundWaveAssetPtr=/Game/Athena/Sounds/Weapons/Fire_Axe_Swing_Distant_01}}
 {id=WavePlayer_22, properties {soundWaveAssetPtr=&PA_Knight_Remix_Swing_01}}
 {id=WavePlayer_23, properties {soundWaveAssetPtr=&/PA_Knight_Remix_Swing_02}}
 {id=WavePlayer_24, properties {soundWaveAssetPtr=&/PA_Knight_Remix_Swing_03}}
 {id=WavePlayer_28, properties {soundWaveAssetPtr=&PA_Knight_Remix_Swing_01}}
 {id=WavePlayer_3, properties {soundWaveAssetPtr=&PA_Knight_Remix_Swing_02}}
 {id=WavePlayer_4, properties {soundWaveAssetPtr=/Game/Athena/Sounds/Weapons/Fire_Axe_Swing_Distant_02}}
 {id=WavePlayer_5, properties {soundWaveAssetPtr=&PA_Knight_Remix_Swing_03}}
 {id=WavePlayer_6, properties {soundWaveAssetPtr=/Game/Athena/Sounds/Weapons/Fire_Axe_Swing_Distant_03}}
 {id=WavePlayer_7, properties {soundWaveAssetPtr=/Game/Athena/Sounds/Weapons/Fire_Axe_Swing_Distant_05}}
 {id=WavePlayer_8, properties {soundWaveAssetPtr=&Fire_Axe_Swing_Distant_06}}
 {id=WavePlayer_9, properties {soundWaveAssetPtr=&Fire_Axe_Swing_Distant_03}}
]
properties {
 audioPropertiesSheet=/BRCosmetics/Sounds/Settings/PropertySheets/Pickaxe/APS_Cosmetic_Pickaxe_Swing
 bOverrideConcurrency=true
 bPrimeOnLoad=true
 maxDistance=3000
 soundClassObject=&Fort_WeaponMelee_PSM
 volumeMultiplier=0.3
 concurrencyOverrides {bLimitToOwner=true, maxCount=2, resolutionRule=StopOldest}
 concurrencySet [&Player_Melee_Weapon_Conc]
}
